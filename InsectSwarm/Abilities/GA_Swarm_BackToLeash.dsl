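; Gameplay ability that sends an InsectSwarm NPC back to its leash location.
; On activation it commits, casts the avatar to NPC_Pawn_Swarm and starts a pathfinding AI move task to that location, using a nav filter that avoids all water.
; It ends when the move finishes with any result, or when the cast or task creation fails.
; On end, it unbinds the move-finished delegate and clears the stored task.
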(event Ability|EventActivateAbility
  (Ability|CommitAbility)
  (bind _asnpc_pawn_swarm (Utilities|Casting|CastToNPC_Pawn_Swarm (Ability|GetAvatarActorfromActorInfo))
    (:then
      (bind _asynctaskproxy (Ability|Tasks|MoveAI "MoveToEQS" 0 (Variables|/InsectSwarm/NPC/Pawns/NPC_Pawn_Swarm.NPC_Pawn_Swarm_C|GetLeashLocation _asnpc_pawn_swarm) "UsePathfinding" "Ignore" true "Default" false true 4e+01 0 "/Game/Athena/AI/NavFilters/NavFilter_NPC_AvoidAllWater.NavFilter_NPC_AvoidAllWater_C" "(TagName=\"\")"))
      (Utilities|IsValid _asynctaskproxy
        (:Is%20Valid
          (bind _output_get (Variables|MovetoEQS|SetMoveTask _asynctaskproxy))
          (EventDispatchers|BindEventtoOnMoveFinished _output_get (delegate Custom|OnMoveFinishedEvent)))
        (:Is%20Not%20Valid
          (Ability|EndAbility))))
    (:CastFailed
      (Ability|EndAbility))))

(event Custom|OnMoveFinishedEvent
  (params (Result :type EPathFollowingResult%20Enum :default "Success") (AIController :type AIController%20Object%20Reference))
  (Ability|EndAbility))

(event Ability|EventOnEndAbility
  (params (bWasCancelled :type Boolean))
  (bind _movetask (Variables|MovetoEQS|GetMoveTask)
    (:then
      (EventDispatchers|UnbindallEventsfromOnMoveFinished _movetask)
      (Variables|MovetoEQS|SetMoveTask 0))
    (:else)))
