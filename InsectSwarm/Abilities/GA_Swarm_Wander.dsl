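; Gameplay ability that drives insect swarm wandering between plants, extending a parent AI move ability.
; On move start, it sphere-traces the destination and marks a tagged attractor plant in use via an environment effect.
; When movement finishes, it traces below the swarm, then either rests on a found plant with an effect and event, or idles randomly before finishing.
; Blocked moves also trigger target selection to attack surroundings and restart movement.
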(event Ability|EventActivateAbility
  (|CallParentFunction) ; This node is disabled and will not be called.\nDrag off pins to build functionality.
)

(event Ability|EventOnEndAbility
  (params (bWasCancelled :type Boolean))
  (|CallParentFunction bWasCancelled) ; This node is disabled and will not be called.\nDrag off pins to build functionality.
)

(event EventStartMoveAI
  ;[1 0.9996 0.2299] Unhooked Leash Stuff. Keeping it in case we want to bring it back.
  (bind _asnpc_pawn_swarm (Utilities|Casting|CastToNPC_Pawn_Swarm (Ability|GetAvatarActorfromActorInfo))
    (:then
      (bind _isinleashradius (Class|NPCPawnSwarm|IsInsideLeashRadius _asnpc_pawn_swarm 0 (Variables|InternalData|GetFirstLocationFound)))
      (if (not _isinleashradius)
        (Ability|EndAbility)))
    (:CastFailed
      _))
  (CallFunction|MarkPlantasinUse)
  (|CallParentFunction))

(event Custom|StartIdleAndEndMoveTo
  (Utilities|FlowControl|Delay (Math|Random|RandomFloatinRange (ScalableFloat|GetValueatLevel (Variables|Default|GetMin_PauseDelayWhenNoPlantsFound)) (ScalableFloat|GetValueatLevel (Variables|Default|GetMax_PauseDelayWhenNoPlantsFound)))) ; If no plants found, stay here for a while...
  (|CallParentFunction (stale EventOnMoveFinishedEvent Result) (stale EventOnMoveFinishedEvent AIController)))

(event EventOnMoveFinishedEvent
  (params (Result :type EPathFollowingResult%20Enum :default "Success") (AIController :type AIController%20Object%20Reference))
  (switch Utilities|FlowControl|Switch|SwitchonEPathFollowingResult Result
    (:Success
      _)
    (:Blocked
      (CallFunction|TryDamageSurroundings))
    (:OffPath
      _)
    (:Aborted
      _)
    (:Invalid
      _))
  (CallFunction|CheckforPlant))

(event Custom|CheckForPlant
  ;[1 1 1] Check if we arrived on a Plant
  (Variables|InternalData|SetbFoundPlant false)
  (bind (_outhits _returnvalue) (Collision|MultiSphereTraceByChannel :Start (Transformation|GetActorLocation (Ability|GetAvatarActorfromActorInfo)) :End (Math|Vector|MakeVector (.x (Transformation|GetActorLocation (Ability|GetAvatarActorfromActorInfo))) (.y (Transformation|GetActorLocation (Ability|GetAvatarActorfromActorInfo))) (- (.z (Transformation|GetActorLocation (Ability|GetAvatarActorfromActorInfo))) 1e+02)) :Radius (ScalableFloat|GetValueatLevel (Variables|Default|GetPlantDecttionRadius)) :TraceChannel "TraceTypeQuery1" :bTraceComplex false :DrawDebugType "None" :bIgnoreSelf true :TraceColor "(R=1.000000,G=0.000000,B=0.000000,A=1.000000)" :TraceHitColor "(R=0.000000,G=1.000000,B=0.000000,A=1.000000)" :DrawTime 15.0))
  (if _returnvalue
    (for _array_element _outhits
      (bind _returnvalue_2 (GameplayTags|HasMatchingGameplayTag(Message) (.HitActor (Collision|BreakHitResult _array_element)) (Variables|Default|GetSwarmAttractorTag)))
      (if _returnvalue_2
        (Variables|Default|SetRestingPlant (.HitActor (Collision|BreakHitResult _array_element)))
        (Variables|InternalData|SetbFoundPlant true)
        (Ability|ApplyGameplayEffectToOwner (Variables|Default|GetGE_WhenOnPlant))
        (Ability|SendGameplayEvent (Variables|Default|GetTag_FoundPlant) (Utilities|Struct|MakeGameplayEventData (Variables|Default|GetTag_FoundPlant) 0 0 (Variables|Default|GetRestingPlant)))
        (|CallParentFunction (stale EventOnMoveFinishedEvent Result) (stale EventOnMoveFinishedEvent AIController))))
    (if (not (Variables|InternalData|GetbFoundPlant))
      (CallFunction|StartIdleandEndMoveTo))
    (else
      (CallFunction|StartIdleandEndMoveTo))))

(event Custom|TryDamageSurroundings
  ;[1 1 1] If Blocked Try to Damage ourselves out
  (bind (_asynctaskproxy _targetdata _applicationtag) (Ability|Tasks|DoTargetSelection "None" "(TagName=\"Targeting.ApplicationTag.0\")")
    (:Targeted
      (if (> (Utilities|Array|Length (Ability|TargetData|GetAllActorsfromTargetData _targetdata)) 0)
        (Ability|ApplyGameplayEffectContainer _targetdata _applicationtag 0)
        (Ability|SendGameplayEvent (Variables|Default|GetTagToSendWhenAttackingWhileBlocked))
        (Ability|ExecuteGameplayCueOnOwner (Variables|Default|GetTag_AttackingCue))
        (CallFunction|StartMoveAI)
        (else
          ;[1 1 1] Check if we arrived on a Plant
          (|CallParentFunction (stale EventOnMoveFinishedEvent Result) (stale EventOnMoveFinishedEvent AIController)))))))

(fn MarkPlantAsInUse
  (bind (_outhits _returnvalue) (Collision|MultiSphereTraceByChannel :Start (+ (Variables|InternalData|GetFirstLocationFound) "0.000000,0.000000,50.000000") :End (- (Variables|InternalData|GetFirstLocationFound) "0.000000,0.000000,100.000000") :Radius (ScalableFloat|GetValueatLevel (Variables|Default|GetPlantDecttionRadius)) :TraceChannel "TraceTypeQuery1" :bTraceComplex false :DrawDebugType "None" :bIgnoreSelf true :TraceColor "(R=1.000000,G=0.000000,B=0.000000,A=1.000000)" :TraceHitColor "(R=0.000000,G=1.000000,B=0.000000,A=1.000000)" :DrawTime 15.0))
  (if _returnvalue
    (for _array_element _outhits
      (bind _returnvalue_2 (GameplayTags|HasMatchingGameplayTag(Message) (.HitActor (Collision|BreakHitResult _array_element)) (Variables|Default|GetSwarmAttractorTag)))
      (if _returnvalue_2
        (Environment|ApplyGlobalEnvironmentGameplayEffecttoActor (.HitActor (Collision|BreakHitResult _array_element)) (Variables|Default|GetGE_PlantInUse))
        (return)))))
